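# Material function for the Lazarus Lens player body that produces a tinted glow color for its tattoos, judging by its naming.
# It boosts the scalar TattooMask input slightly with CheapContrast, then multiplies it by the Tattoo_EmissiveColor parameter, a cyan-blue by default.
# The Tattoo_EmissivePhase scalar scales the result and defaults to zero, so the output is black until that parameter is set.
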MaterialFunction /Game/Characters/Player/Male/Medium/Bodies/M_MED_LazarusLens/Materials/MF_LazarusLens_TattooGlow
graph {nodes [
  n1 = MaterialFunctionCall(In=FunctionInput() {inputName=TattooMask, inputType=FunctionInput_Scalar}, Contrast=Constant() {r=0.1}) {materialFunction=&CheapContrast}
  n2 = VectorParameter() {parameterName=Tattoo_EmissiveColor, defaultValue=<rgba 0.05 0.5996 1 1>}
  FunctionOutput(None=Multiply(A=Multiply(A=@n1.Result, B=@n2.RGB), B=ScalarParameter() {parameterName=Tattoo_EmissivePhase}))
 ]}
